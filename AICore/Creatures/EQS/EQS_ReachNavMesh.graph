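# EQS query that picks a nearby navmesh point for a creature to move to, apparently to reach the navmesh.
# A donut of points around the querier is projected onto the navmesh through the NavFilter_ReachNavMesh filter, with a large vertical search range.
# Inverse-linear 3D distance scoring favors closer points, while a Dot test with a capped score discourages 180-degree turns.
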EnvQuery /AICore/Creatures/EQS/EQS_ReachNavMesh
options [{
  generator {
   name=EnvQueryGenerator_Donut_0
   class=EnvQueryGenerator_Donut
   properties {
    bUseSpiralPattern=false
    center=&EnvQueryContext_Querier
    innerRadius=200
    numberOfRings=3
    outerRadius=2000
    pointsPerRing=8
    projectionData {
     extentX=100
     navigationFilter=/AICore/Creatures/NavFilters/NavFilter_ReachNavMesh.NavFilter_ReachNavMesh_C
     postProjectionVerticalOffset=0
     projectDown=2000
     projectUp=2000
    }
   }
  }
  tests [
   EnvQueryTest_Distance_1 {
    class=EnvQueryTest_Distance
    properties {
     distanceTo=&EnvQueryContext_Querier
     filterType=Range
     floatValueMin=0
     scoringEquation=InverseLinear
     scoringFactor=1
     testComment=""
     testMode=Distance3D
     testPurpose=Score
    }
   }
   EnvQueryTest_Dot_1 {
    class=EnvQueryTest_Dot
    properties {clampMaxType=SpecifiedValue, clampMinType=None, scoreClampMax=0.5, scoreClampMin=0, scoringFactor=1, testComment=Avoid 180 degrees turns}
   }
  ]
 }]
